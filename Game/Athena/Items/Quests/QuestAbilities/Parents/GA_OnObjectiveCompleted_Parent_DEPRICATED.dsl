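; Deprecated parent gameplay ability that lets child abilities react when the avatar's player completes objectives of a configured quest.
; On activation it records that quest's objective backend names, pre-fills the already completed set, and binds to the quest manager's dynamic quest update event.
; Each update filters for newly completed objectives of that quest, adds them to the set, and calls the empty OnObjectiveCompleted hook with completion count and name.
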(event Ability|EventActivateAbility
  (Default|SetupQuestObjectives)
  (Quests|BindEventtoOnDisplayDynamicQuestUpdate (Quests|GetQuestManager (Fortnite|GetFortPlayerControllerfromActor (CallFunction|GetAvatarasFortPlayerPawn)) "Invalid") (delegate Custom|OnQuestsUpdated)))

(event Custom|OnQuestsUpdated
  (params (QuestObjective :type Fort%20Quest%20Objective%20Info%20Object%20Reference) (bDisplayStatusUpdate :type Boolean) (bDisplayAnnouncementUpdate :type Boolean))
  (Default|HandleObjectiveCompleted QuestObjective))

(event Ability|EventFailedToActivatePassiveAbility
  (params (ActorInfo :type Gameplay%20Ability%20Actor%20Info%20Structure)))

(fn HandleObjectiveCompleted
  (params (QuestObjective :type Fort%20Quest%20Objective%20Info%20Object%20Reference))
  (bind (_isobjectivecompleted _objectivebackendname) (Default|HasCompletedObjective QuestObjective))
  (if _isobjectivecompleted
    (Utilities|Set|Add (Variables|Data|GetCompletedObjectives) _objectivebackendname)
    (CallFunction|OnObjectiveCompleted (Utilities|Set|Length (Variables|Data|GetCompletedObjectives)) _objectivebackendname)))

(fn HasCompletedObjective
  (params (QuestObjective :type Fort%20Quest%20Objective%20Info%20Object%20Reference))
  (return (Math|Boolean|ANDBoolean (Quest|HasCompleted QuestObjective) (Utilities|Set|ContainsItem (Variables|Data|GetQuestObjectives) (Variables|/Script/FortniteGame.FortQuestObjectiveInfo|GetBackendName QuestObjective)) (not (Utilities|Set|ContainsItem (Variables|Data|GetCompletedObjectives) (Variables|/Script/FortniteGame.FortQuestObjectiveInfo|GetBackendName QuestObjective))) (== (Quest|GetQuestDefinition (Quest|GetOwningQuest QuestObjective)) (Variables|Setup|GetQuestItemDefinition))) (Variables|/Script/FortniteGame.FortQuestObjectiveInfo|GetBackendName QuestObjective)))

(fn OnObjectiveCompleted
  (params (ObjectiveCompletionCount :type Integer) (ObjectiveBackendName :type Name)))

(fn SetupQuestObjectives
  (for _array_element (QuestItemDefinition|GetObjectives (Variables|Setup|GetQuestItemDefinition))
    (Utilities|Set|Add (Variables|Data|GetQuestObjectives) (Utilities|Struct|BreakFortMcpQuestObjectiveInfo _array_element))
    (bind (_playercontrollerout _completedquestobjective) (Quest|HasPlayerCompletedQuestObjective (Fortnite|GetFortPlayerControllerfromActor (CallFunction|GetAvatarasFortPlayerPawn)) (Variables|Setup|GetQuestItemDefinition) (Utilities|Struct|BreakFortMcpQuestObjectiveInfo _array_element)))
    (if _completedquestobjective
      (Utilities|Set|Add (Variables|Data|GetCompletedObjectives) (Utilities|Struct|BreakFortMcpQuestObjectiveInfo _array_element)))))
