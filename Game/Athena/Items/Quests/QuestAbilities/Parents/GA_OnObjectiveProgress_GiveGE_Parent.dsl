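; Parent quest gameplay ability that grants the owner a gameplay effect as an objective is completed.
; It overrides OnObjectiveCompleted, calls the parent, then looks up the completion count in the CompletionCountToGEs setup map.
; If that count has an entry, the mapped gameplay effect is applied to the owner; activate and end events add no logic.
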(event Ability|EventActivateAbility
  (|CallParentFunction) ; This node is disabled and will not be called.\nDrag off pins to build functionality.
)

(event Ability|EventOnEndAbility
  (params (bWasCancelled :type Boolean)))

(fn OnObjectiveCompleted
  (params (ObjectiveCompletionCount :type Integer) (ObjectiveBackendName :type Name :default "None"))
  (bind _completioncounttoges (Variables|Setup|GetCompletionCountToGEs))
  (bind (_value _returnvalue) (Utilities|Map|Find _completioncounttoges ObjectiveCompletionCount))
  (|CallParentFunction ObjectiveCompletionCount ObjectiveBackendName)
  (if _returnvalue
    (Ability|ApplyGameplayEffectToOwner _value)))
