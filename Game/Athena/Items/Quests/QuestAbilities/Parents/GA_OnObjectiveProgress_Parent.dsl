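; Parent gameplay ability that lets child abilities react as objectives of a configured quest complete.
; On activation it records the quest's objectives and those already completed, checking via the quest component or the QS1 quest manager, depending on the definition.
; The completion event filters for new completions from that quest, then passes the completed count and backend name to the empty OnObjectiveCompleted hook.
; No binding of that event is shown.
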(event Ability|EventActivateAbility
  (|CallParentFunction)
  (CallFunction|SetupQuestObjectives)
  (bind _returnvalue (Quest|GetQuestComponentforPlayerObject (CallFunction|GetAvatarasFortPlayerPawn)))
  (Utilities|IsValid _returnvalue
    (:Is%20Valid
      (bind _returnvalue_2 (Quest|FindQuestInstanceforParticipant _returnvalue (Item|GetPersistentName (Variables|Setup|GetQuestItemDefinition))))
      (bind _asfort_quest_item (Utilities|Casting|CastToFortQuestItem _returnvalue_2))
      (Utilities|IsValid _asfort_quest_item))))

(event Ability|EventFailedToActivatePassiveAbility
  (params (ActorInfo :type Gameplay%20Ability%20Actor%20Info%20Structure)))

(event Custom|OnQuestObjectiveCompleted_Event
  (params (ObjectiveInstance :type Fort%20Quest%20Objective%20Info%20Object%20Reference))
  (CallFunction|HandleObjectiveCompleted ObjectiveInstance))

(fn HandleObjectiveCompleted
  (params (QuestObjective :type Fort%20Quest%20Objective%20Info%20Object%20Reference))
  (bind (_isobjectivecompleted _objectivebackendname) (CallFunction|HasCompletedObjective QuestObjective))
  (if _isobjectivecompleted
    (Utilities|Set|Add (Variables|Data|GetCompletedObjectives) _objectivebackendname)
    (CallFunction|OnObjectiveCompleted (Utilities|Set|Length (Variables|Data|GetCompletedObjectives)) _objectivebackendname)))

(fn HasCompletedObjective
  (params (QuestObjective :type Fort%20Quest%20Objective%20Info%20Object%20Reference))
  (return (Math|Boolean|ANDBoolean (Quest|HasCompleted QuestObjective) (Utilities|Set|ContainsItem (Variables|Data|GetQuestObjectives) (Variables|/Script/FortniteGame.FortQuestObjectiveInfo|GetBackendName QuestObjective)) (not (Utilities|Set|ContainsItem (Variables|Data|GetCompletedObjectives) (Variables|/Script/FortniteGame.FortQuestObjectiveInfo|GetBackendName QuestObjective))) (== (Quest|GetQuestDefinition (Quest|GetOwningQuest QuestObjective)) (Variables|Setup|GetQuestItemDefinition))) (Variables|/Script/FortniteGame.FortQuestObjectiveInfo|GetBackendName QuestObjective)))

(fn HasCompletedObjective_QS1
  (params (BackendName :type Name) (Definition :type Fort%20Quest%20Item%20Definition%20Object%20Reference))
  (return (Quest|HasCompletedObjectivewithName (Quests|GetQuestwithDefinition (Quests|GetQuestManager (Fortnite|GetFortPlayerControllerfromActor (CallFunction|GetAvatarasFortPlayerPawn)) "Invalid") Definition) BackendName :!class /Script/FortniteGame.FortQuestItem)))

(fn HasCompletedQuestObjective
  (params (QuestDefinition :type Fort%20Quest%20Item%20Definition%20Object%20Reference) (ObjectiveName :type Name))
  (if (Class|FortQuestItemDefinition|HasComponentforWorldProcessingUpdates QuestDefinition)
    (bind _returnvalue (Quest|GetQuestComponentforPlayerObject (CallFunction|GetAvatarasFortPlayerPawn)))
    (Utilities|IsValid _returnvalue
      (:Is%20Valid
        (bind _returnvalue_2 (Quest|FindQuestInstanceforParticipant _returnvalue (Item|GetPersistentName QuestDefinition)))
        (Utilities|IsValid (Utilities|ToObject(Interface) _returnvalue_2)
          (:Is%20Valid
            (return (Class|QuestInstanceInterface|HasCompletedObjectiveInstance _returnvalue_2 ObjectiveName)))
          (:Is%20Not%20Valid
            _)))
      (:Is%20Not%20Valid
        _))
    (return false)
    (else
      (bind _returnvalue_3 (CallFunction|HasCompletedObjectiveQS1 ObjectiveName QuestDefinition))
      (return _returnvalue_3))))

(fn OnObjectiveCompleted
  (params (ObjectiveCompletionCount :type Integer) (ObjectiveBackendName :type Name)))

(fn SetupQuestObjectives
  (for _array_element (QuestItemDefinition|GetObjectives (Variables|Setup|GetQuestItemDefinition))
    (Utilities|Set|Add (Variables|Data|GetQuestObjectives) (Utilities|Struct|BreakFortMcpQuestObjectiveInfo _array_element))
    (bind _objectivecompleted (CallFunction|HasCompletedQuestObjective (Variables|Setup|GetQuestItemDefinition) (Utilities|Struct|BreakFortMcpQuestObjectiveInfo _array_element)))
    (if _objectivecompleted
      (Utilities|Set|Add (Variables|Data|GetCompletedObjectives) (Utilities|Struct|BreakFortMcpQuestObjectiveInfo _array_element)))))
