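; Parent quest ability that counts ignite events on targets matching required and excluded gameplay tags, for quest progress.
; A target tag check gates activation, then a WaitGameplayEvent listener repeats that check plus an optional same-team test for each event.
; Each counted event increments the objective count and restarts a timeout that ends the ability.
; On end, it sends the batched count as one tagged "Ignite" custom stat event to the quest manager, then clears it.
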(event Ability|EventActivateAbilityFromEvent
  (params (EventData :type Gameplay%20Event%20Data%20Structure%20%28by%20ref%29))
  (|CallParentFunction EventData)
  (CallFunction|ResetAbilityTimer)
  ;[1 1 1] Wait until an ignite event occurs on a target with the required target tags. This can occur many times.
  (bind (_asynctaskproxy _payload) (Ability|Tasks|WaitGameplayEvent (Variables|GAVars|Data|GetT_TriggerTag))
    (:EventReceived
      (bind _result (CallFunction|MeetsTagRequirements (.Target (Utilities|Struct|BreakGameplayEventData _payload))))
      (if _result
        (if (not (Variables|GAVars|Setup|GetbCheckSameTeam))
          _
          (elif (Team|OnSameTeam (.Instigator (Utilities|Struct|BreakGameplayEventData _payload)) (.Target (Utilities|Struct|BreakGameplayEventData _payload)))
            _))
        ;[1 1 1] Increment the objective count and reset the timer to wait for more potential ignite events.
        (Math|Integer|IncrementInt (Variables|GAVars|Data|GetObjectiveCount)
          (:%20%20
            (CallFunction|ResetAbilityTimer)))))))

(event Custom|ResetAbilityTimer
  ;[1 1 1] Timer that ends the ability after no ignite events have occurred  before the timeout.
  (Utilities|Time|ClearandInvalidateTimerbyHandle (Variables|GAVars|Data|GetTimer_AbilityEnd))
  (bind _returnvalue (Utilities|Time|SetTimerbyEvent (delegate Custom|CustomEndAbility) (Variables|GAVars|Setup|GetTimeout)))
  (Variables|GAVars|Data|SetTimer_AbilityEnd _returnvalue))

(event Custom|CustomEndAbility
  ;[1 1 1] Timer that ends the ability after no ignite events have occurred  before the timeout.
  (Ability|EndAbility))

(event Ability|EventOnEndAbility
  (params (bWasCancelled :type Boolean))
  ;[1 1 1] When the ability ends for any reason, we send off a complex custom stat event with an objective count equal to the number of times the ability was triggered, then we reset the counter.
  (|CallParentFunction bWasCancelled)
  (Quest|SendCustomStatEventwithTags (Variables|GAVars|Data|GetOwningPlayerQuestManager) "Ignite" (Variables|GAVars|Data|GetTC_Ignite) (Variables|GAVars|Setup|GetTC_TargetTags_0) (Variables|GAVars|Data|GetObjectiveCount))
  (Variables|GAVars|Data|SetObjectiveCount 0))

(fn ShouldAbilityRespondToEvent
  (params (ActorInfo :type Gameplay%20Ability%20Actor%20Info%20Structure) (Payload_EventTag :type Gameplay%20Tag%20Structure) (Payload_Instigator :type Actor%20Object%20Reference) (Payload_Target :type Actor%20Object%20Reference) (Payload_OptionalObject :type Object%20Reference) (Payload_OptionalObject2 :type Object%20Reference) (Payload_ContextHandle :type Gameplay%20Effect%20Context%20Handle%20Structure) (Payload_InstigatorTags :type Gameplay%20Tag%20Container%20Structure) (Payload_TargetTags :type Gameplay%20Tag%20Container%20Structure) (Payload_EventMagnitude :type Float%20%28single-precision%29) (Payload_TargetData :type Gameplay%20Ability%20Target%20Data%20Handle%20Structure) (Payload_InstancedEventData :type Instanced%20Struct%20Structure))
  ;[1 1 1] If the ignited target has the required target tags, trigger the ability.
  (bind _result (CallFunction|MeetsTagRequirements Payload_Target))
  (return _result))

(fn MeetsTagRequirements
  (params (TargetObject :type Actor%20Object%20Reference))
  (bind _returnvalue (GameplayTags|GetOwnedGameplayTags TargetObject :!class /Script/GameplayTags.GameplayTagAssetInterface))
  (return (and (GameplayTags|HasAnyTags _returnvalue (Variables|GAVars|Setup|GetTC_TargetTags_0)) (not (GameplayTags|HasAnyTags _returnvalue (Variables|GAVars|Setup|GetTC_ExcludeTags))))))
