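; Parent gameplay ability for passive quest abilities that caches the owning player pawn and its quest manager.
; On activation and on pawn change, it stores the current player pawn, resolves that pawn's player controller and, if valid, saves the controller's quest manager.
; The failed-activation event is empty, and the graph holds no quest-specific logic.
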(event Ability|EventActivateAbility
  (bind _fortplayerpawn (CallFunction|GetCurrentPlayerPawn)
    (:Valid
      ;[1 1 1] Save pawn and quest manager for easy reference
      (bind _output_get (Variables|GAVars|Data|SetOwningPlayerPawn _fortplayerpawn))
      (if (Utilities|IsValid (Fortnite|GetFortPlayerControllerfromActor _output_get) :!class /Script/Engine.KismetSystemLibrary)
        (Variables|GAVars|Data|SetOwningPlayerQuestManager (Quests|GetQuestManager (Fortnite|GetFortPlayerControllerfromActor _output_get) "Invalid"))))))

(event EventBP_OnPawnChanged
  (bind _fortplayerpawn (CallFunction|GetCurrentPlayerPawn)
    (:Valid
      ;[1 1 1] Save pawn and quest manager for easy reference
      (bind _output_get (Variables|GAVars|Data|SetOwningPlayerPawn _fortplayerpawn))
      (if (Utilities|IsValid (Fortnite|GetFortPlayerControllerfromActor _output_get) :!class /Script/Engine.KismetSystemLibrary)
        (Variables|GAVars|Data|SetOwningPlayerQuestManager (Quests|GetQuestManager (Fortnite|GetFortPlayerControllerfromActor _output_get) "Invalid"))))))

(event Ability|EventFailedToActivatePassiveAbility
  (params (ActorInfo :type Gameplay%20Ability%20Actor%20Info%20Structure)))
