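; Parent quest gameplay ability that, when triggered by a gameplay event, removes gameplay effects from its owner.
; It calls the parent activation, then removes the effects immediately or after a configurable removal delay via a stored timer.
; Removal loops over the configured asset tags, removing matching effects for each, then ends the ability.
; Ending the ability clears and invalidates the timer, so a pending removal never fires afterwards.
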(event Ability|EventActivateAbilityFromEvent
  (params (EventData :type Gameplay%20Event%20Data%20Structure%20%28by%20ref%29))
  (bind _removaldelay (Variables|GAVars|Setup|GetRemovalDelay))
  (|CallParentFunction EventData)
  (if (<= _removaldelay 0.0)
    (CallFunction|RemoveGEswithTags)
    (else
      (bind _returnvalue (Utilities|Time|SetTimerbyEvent (EventDispatchers|CreateEvent :!event Custom|RemoveGEsWithTags) _removaldelay))
      (Variables|Default|SetTimer_RemoveGEs _returnvalue))))

(event Custom|RemoveGEsWithTags
  (for _array_element (Variables|GAVars|Setup|GetAssetTagsOfGEsToBeRemoved)
    (Ability|RemoveGameplayEffectFromOwnerWithAssetTags _array_element))
  (Ability|EndAbility))

(event Ability|EventOnEndAbility
  (params (bWasCancelled :type Boolean))
  (Utilities|Time|ClearandInvalidateTimerbyHandle (Variables|Default|GetTimer_RemoveGEs)))
