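; Parent gameplay ability for event-triggered quest abilities, which caches the owning player's pawn and quest manager for easy reference.
; On activation from an event and on pawn change, it runs identical logic: get the current player pawn and, if valid, store it.
; It then resolves that pawn's player controller and, if valid, stores its quest manager.
; The event data is unused, and no quest-specific logic appears here.
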(event Ability|EventActivateAbilityFromEvent
  (params (EventData :type Gameplay%20Event%20Data%20Structure%20%28by%20ref%29))
  (bind _fortplayerpawn (CallFunction|GetCurrentPlayerPawn)
    (:Valid
      ;[1 1 1] Save pawn and quest manager for easy reference
      (bind _output_get (Variables|GAVars|Data|SetOwningPlayerPawn _fortplayerpawn))
      (if (Utilities|IsValid (Fortnite|GetFortPlayerControllerfromActor _output_get) :!class /Script/Engine.KismetSystemLibrary)
        (Variables|GAVars|Data|SetOwningPlayerQuestManager (Quests|GetQuestManager (Fortnite|GetFortPlayerControllerfromActor _output_get) "Invalid"))))))

(event EventBP_OnPawnChanged
  (bind _fortplayerpawn (CallFunction|GetCurrentPlayerPawn)
    (:Valid
      ;[1 1 1] Save pawn and quest manager for easy reference
      (bind _output_get (Variables|GAVars|Data|SetOwningPlayerPawn _fortplayerpawn))
      (if (Utilities|IsValid (Fortnite|GetFortPlayerControllerfromActor _output_get) :!class /Script/Engine.KismetSystemLibrary)
        (Variables|GAVars|Data|SetOwningPlayerQuestManager (Quests|GetQuestManager (Fortnite|GetFortPlayerControllerfromActor _output_get) "Invalid"))))))
